WidgetBlueprint /VoiceRep/VoiceReportWizard/StepWizard/WBP_StepWizard_Base
graphs [
 Add Step {nodes [
   n1 = Add Step!() -> (then, Step:int) {is=title K2Node_FunctionEntry}
   > Set Step Num(Step Num:int=@n1) -> (Output_Get:int=0)
   > n2 = Construct Object from Class(Class=&/BP_UI_StepWizard.BP_UI_StepWizard_C) -> (ReturnValue:BP UI Step Wizard Object Reference)
   > n3 = Set NextListItem(NextListItem:BP UI Step Wizard Object Reference=@n2) -> (Output_Get:BP UI Step Wizard Object Reference)
   n4 = "Get (a copy)"(
    Array:Array of S UI Step Wizard Entry Structures=Get CurrentStepList() -> (CurrentStepList:Array of S UI Step Wizard Entry Structures)
    Dimension 1=Get Step Num() -> (Step Num:int=0)
   ) -> (Output:S UI Step Wizard Entry Structure)
   > Set StepContentData!(execute, StepContentData:S UI Step Wizard Entry Structure=@n4, self:BP UI Step Wizard Object Reference=@n3) -> (then, Output_Get:S UI Step Wizard Entry Structure)
   > Set CurrentStep!(
    execute
    CurrentStep:int=Get Step Num() -> (Step Num=0)
    self:BP UI Step Wizard Object Reference=Get NextListItem() -> (NextListItem:BP UI Step Wizard Object Reference)
   ) -> (then, Output_Get:int=0)
   > Set TotalSteps!(execute, TotalSteps:int=Get TotalSteps() -> (TotalSteps:int=0), self:BP UI Step Wizard Object Reference=Get NextListItem()) -> (then, Output_Get:int=0)
   > AddItem!(execute, self:List View Object Reference=Get ListView() -> (ListView:List View Object Reference), Item:Object Reference=Get NextListItem()) -> (then)
   > Get Current Step Name!(execute, self, Step:int=Get Step Num() -> (Step Num=0)) -> (then) {is=title K2Node_CallFunction}
   n5 = Break S UI Step Wizard Entry!(S_UI_StepWizard_Entry:"S UI Step Wizard Entry Structure (by ref)"=@n4) -> (
    StepWidgetContent_92_#1:Common Activatable Widget Class Reference
    StepType_76_#2:E_UI_StepWizard_Type Enum
    CurrentStepNameText_82_#3:Text
    ButtonBack_90_#4:bool=~
    ButtonClose_91_#5:bool=~
   ) {is=title K2Node_BreakStruct}
   > SetVisibility(
    self=Get Button_Back() -> (Button_Back:WBP UIKit Button Quiet Object Reference)
    InVisibility=Select(Option 0=Collapsed, Option 1=Visible, Index=@n5.ButtonBack_90_#4)
   )
   > SetVisibility(
    self=[
     Get Button_TouchClose() -> (Button_TouchClose:Close Button Object Reference)
     Get Button_Close() -> (Button_Close:WBP UIKit Button Quiet Object Reference)
    ]
    InVisibility=Select(Option 0=Collapsed, Option 1=Visible, Index=@n5.ButtonClose_91_#5)
   )
   n6 = GetNumItems!(self:List View Object Reference=Get ListView()) -> (ReturnValue:int=0)
   > SetScrollOffset!(execute, self:List View Base Object Reference=Get ListView(), InScrollOffset:float="To Float (Integer)"(InInt=int - int(A=@n6))) -> (then)
   > Get Next Step Data!(execute, self, Current Step:int=Get Step Num() -> (Step Num=0)) -> (then) {is=title K2Node_CallFunction}
  ]}
 EventGraph {nodes [
   "On Clicked (Button_Back)"!() -> (OutputDelegate, then, Button:Common Button Base Object Reference) {is=title K2Node_ComponentBoundEvent}
   > Previous Step()
   n1 = "On Clicked (Button_Close)"!() -> (OutputDelegate, then, Button:Common Button Base Object Reference) {is=title K2Node_ComponentBoundEvent}
   RemoveFromParent(execute=[@n1.then, @n2.then, @n5.else], self=Get Parent Widget() -> (Parent Widget:Widget Object Reference))
   > HandleCloseButtonClicked(self:Voice Rep Modal Panel Container Object Reference)
   n2 = "On Clicked (Button_TouchClose)"!() -> (OutputDelegate, then, Button:Common Button Base Object Reference) {is=title K2Node_ComponentBoundEvent}
   n3 = "On Entry Initialized (ListView)"!() -> (OutputDelegate, then, Item:Object Reference, Widget:User Widget Object Reference) {is=title K2Node_ComponentBoundEvent}
   > n4 = Cast To WBP_StepWizard_List_Entry!(execute, Object=@n3.Widget) -> (then, CastFailed, AsWBP Step Wizard List Entry:WBP Step Wizard List Entry Object Reference) {is=title K2Node_DynamicCast}
   > Set CurrentDataWidget(CurrentDataWidget:Widget Object Reference=Get ContentWidget!(self:WBP Step Wizard List Entry Object Reference=@n4) -> (ContentWidget:User Widget Object Reference)) -> (Output_Get:Widget Object Reference)
   Previous Step() {is=title K2Node_CustomEvent}
   > n5 = Branch(Condition="integer > integer"(A=Get CurrentStep() -> (CurrentStep:int=0)))
   > n6 = Decrement Int(Value=Get CurrentStep() -> (CurrentStep=0))
   > SetScrollOffset(self=Get ListView(), InScrollOffset="To Float (Integer)"(InInt=@n6))
   Next Step() {is=title K2Node_CustomEvent}
   n7 = Get CurrentStepList()
   n8 = Get CurrentStep() -> (CurrentStep=0)
   > Branch(Condition="integer > integer"(A=Length(TargetArray:Array of S UI Step Wizard Entry Structures=@n7), B=int + int(A=@n8)))
   > n9 = Increment Int(Value=Get CurrentStep() -> (CurrentStep=0))
   > Add Step!(execute, self, Step:int=@n9) -> (then) {is=title K2Node_CallFunction}
   n10 = Create Step Wizard!() -> (OutputDelegate, then, List Name:String, Parent Widget:Widget Object Reference) {is=title K2Node_CustomEvent}
   > Set Parent Widget(Parent Widget:Widget Object Reference=@n10.Parent Widget) -> (Output_Get:Widget Object Reference)
   n11 = Find(
    TargetMap:Map of Strings to S UI Step Wizard Step List Structures=Get StepList() -> (StepList:Map of Strings to S UI Step Wizard Step List Structures)
    Key=@n10.List Name
   ) -> (Value:S UI Step Wizard Step List Structure)
   > Branch(Condition=@n11.ReturnValue)
   > ClearListItems(self=Get ListView())
   n12 = Break S UI Step Wizard Step List!(S_UI_StepWizard_StepList:"S UI Step Wizard Step List Structure (by ref)"=@n11) -> (
    StepWizardList_71_#6:Array of S UI Step Wizard Entry Structures
    StepWizardIcon_78_#7:Material Instance Object Reference
    HeaderText_76_#8:Text
   ) {is=title K2Node_BreakStruct}
   > Set CurrentStepList(CurrentStepList:Array of S UI Step Wizard Entry Structures=@n12) -> (Output_Get:Array of S UI Step Wizard Entry Structures)
   > SetBrushFromMaterial(self=Get Report_Icon() -> (Report_Icon:Image Object Reference), Material=@n12.StepWizardIcon_78_#7)
   > "SetText (Text)"(self=Get HeaderText() -> (HeaderText:Common Text Object Reference), InText=@n12.HeaderText_76_#8)
   > Get Total Steps() {is=title K2Node_CallFunction}
   > n13 = Set CurrentStep!(execute, CurrentStep=0) -> (then, Output_Get=0)
   > Add Step(Step=@n13)
   n14 = Load Step From List!() -> (OutputDelegate, then, List Name:String, Step:int) {is=title K2Node_CustomEvent}
   n15 = Find(TargetMap=Get StepList(), Key=@n14.List Name)
   > Branch(Condition=@n15.ReturnValue)
   n16 = Break S UI Step Wizard Step List(S_UI_StepWizard_StepList=@n15)
   > Set CurrentStepList(CurrentStepList=@n16)
   > SetBrushFromMaterial(self=Get Report_Icon(), Material=@n16.StepWizardIcon_78_#7)
   > "SetText (Text)"(self=Get HeaderText(), InText=@n16.HeaderText_76_#8)
   > Add Step(Step=@n14.Step)
   Event Construct()
   n17 = IsMobileGame()
   > SetDesiredSizeOverride!(
    execute
    self:Image Object Reference=Get Step_Image() -> (Step_Image:Image Object Reference)
    DesiredSize:Vector 2D Structure=Select(Option 0:Vector 2D Structure=(X=16,Y=16), Option 1:Vector 2D Structure=(X=24,Y=24), Index=@n17) -> (ReturnValue:Vector 2D Structure)
   ) -> (then)
   > SetDesiredSizeOverride(self=Get Report_Icon(), DesiredSize=Select(Option 0=(X=64,Y=64), Option 1=(X=96,Y=96), Index=@n17))
   Event SubmitReport() {is=title K2Node_Event}
   > n18 = Cast To WBP_VoiceRep_Reporting_Step!(execute, Object=Get CurrentDataWidget() -> (CurrentDataWidget:Widget Object Reference)) -> (then, CastFailed, AsWBP Voice Rep Reporting Step:WBP Voice Rep Reporting Step Object Reference) {is=title K2Node_DynamicCast}
   > OnSubmitReport(self:WBP Voice Rep Reporting Step Object Reference=@n18)
  ]}
 Get Current Step Name {nodes [
   n1 = Get Current Step Name!() -> (then, Step:int) {is=title K2Node_FunctionEntry}
   n2 = Get CurrentStepList()
   > Set Text(
    Text=Break S UI Step Wizard Entry(S_UI_StepWizard_Entry="Get (a copy)"(Array=@n2, Dimension 1=@n1)) -> (ButtonBack_90_#4=~, ButtonClose_91_#5=~).CurrentStepNameText_82_#3
    self:Rich Text Block Object Reference=Get CurrentStepNameText() -> (CurrentStepNameText:Common Rich Text Block Object Reference)
   )
  ]}
 Get Next Step Data {nodes [
   n1 = Get Next Step Data!() -> (then, Current Step:int) {is=title K2Node_FunctionEntry}
   > Set NextStepNum(NextStepNum:int=int + int(A=@n1)) -> (Output_Get:int=0)
   n2 = Get CurrentStepList()
   > n3 = Branch(Condition="integer >= integer"(A=Get NextStepNum() -> (NextStepNum:int=0), B=Length(TargetArray=@n2)))
   > SetVisibility(self=Get BNextStep() -> (BNextStep:Border Object Reference), InVisibility=Hidden)
   SetVisibility(execute=@n3.else, self=Get BNextStep(), InVisibility=Visible)
   n4 = Get CurrentStepList()
   n5 = Get NextStepNum() -> (NextStepNum=0)
   > n6 = Switch on E_UI_StepWizard_Type!(
    execute
    Selection:E_UI_StepWizard_Type Enum=Break S UI Step Wizard Entry(S_UI_StepWizard_Entry="Get (a copy)"(Array=@n4, Dimension 1=@n5)) -> (ButtonBack_90_#4=~, ButtonClose_91_#5=~).StepType_76_#2
   ) -> (NewEnumerator0:Exec, NewEnumerator1:Exec, NewEnumerator2:Exec) {is=title K2Node_SwitchEnum}
   > "SetText (Text)"(self=Get Next_Step_Name() -> (Next_Step_Name:Common Text Object Reference), InText="LOCTEXT(\"Step\")")
   > SetVisibility(self=Get CountBoxNext() -> (CountBoxNext:Horizontal Box Object Reference), InVisibility=Visible)
   > "SetText (Text)"(
    self=Get NextStepText() -> (NextStepText:Common Text Object Reference)
    InText="To Text (Integer)"(Value=Get NextStepNum() -> (NextStepNum=0))
   )
   > "SetText (Text)"(
    self=Get TotalStepsText() -> (TotalStepsText:Common Text Object Reference)
    InText="To Text (Integer)"(Value=Get TotalSteps() -> (TotalSteps=0))
   )
   "SetText (Text)"(execute=@n6.NewEnumerator1, self=Get Next_Step_Name(), InText="LOCTEXT(\"Thanks\")")
   > SetVisibility(self=Get CountBoxNext(), InVisibility=Collapsed)
   "SetText (Text)"(execute=@n6.NewEnumerator2, self=Get Next_Step_Name(), InText="LOCTEXT(\"Ups!\")")
   > SetVisibility(self=Get CountBoxNext(), InVisibility=Collapsed)
  ]}
 Get Total Steps {nodes [
   Get Total Steps() {is=title K2Node_FunctionEntry}
   > Set TotalSteps!(execute, TotalSteps=0) -> (then, Output_Get=0)
   > n1 = For Each Loop(Array:Array of S UI Step Wizard Entry Structures=Get CurrentStepList()) -> (Array Element:S UI Step Wizard Entry Structure)
   > Switch on E_UI_StepWizard_Type(Selection=Break S UI Step Wizard Entry(S_UI_StepWizard_Entry=@n1) -> (ButtonBack_90_#4=~, ButtonClose_91_#5=~).StepType_76_#2)
   > Increment Int(Value=Get TotalSteps() -> (TotalSteps=0))
  ]}
]
variables [
 CurrentDataWidget {type=Widget, default=null}
 CurrentListItem {type=BP_UI_StepWizard_C, default=null}
 CurrentStep {type=integer}
 CurrentStepList {type="array<S_UI_StepWizard_Entry>", default []}
 NextListItem {type=BP_UI_StepWizard_C, default=null}
 Parent Widget {type=Widget, default=null}
 Step Type {type=E_UI_StepWizard_Type, default=Step}
 StepList {type="map<S_UI_StepWizard_StepList>", default {}}
 TotalSteps {type=integer}
]
widgets {
 info {inheritedWidgetCount=0, parentClass=/Script/FortniteUI.VoiceRepModalPanelContainer, rootWidgetClass=&Overlay, widgetCount=26}
 widgets [
  Overlay {widgetClassPath=&Overlay, widget=Overlay}
  VBContent {
   widgetClassPath=&VerticalBox
   parent=Overlay
   widget=VBContent
   slot {class=OverlaySlot, properties {horizontalAlignment=HAlign_Fill, verticalAlignment=VAlign_Fill}}
  }
  HBHeader {
   widgetClassPath=&HorizontalBox
   parent=VBContent
   widget=HBHeader
   slot {class=VerticalBoxSlot, properties {horizontalAlignment=HAlign_Left, verticalAlignment=VAlign_Top, padding=<ltrb 120 80 120 30>}}
  }
  Border_69 {
   widgetClassPath=&Border
   parent=HBHeader
   widget=Border_69
   properties {
    horizontalAlignment=HAlign_Left
    verticalAlignment=VAlign_Top
    background {
     bIsDynamicallyLoaded=false
     drawAs=Image
     imageType=NoImage
     mirroring=NoMirror
     resourceName=None
     resourceObject=/VoiceRep/Materials/MI_UI_StepIcon_Background
     tiling=NoTile
     imageSize=<32 32>
     margin=<ltrb 0 0 0 0>
     outlineSettings {
      bUseBrushTransparency=false
      roundingType=HalfHeightRadius
      width=0
      color {colorUseRule=UseColor_Specified, specifiedColor=<rgba 0 0 0 0>}
      cornerRadii=<0 0 0 0>
     }
     tintColor {colorUseRule=UseColor_Specified, specifiedColor=<rgba 1 1 1 1>}
     uVRegion {bIsValid=false, max=<0 0>, min=<0 0>}
    }
    padding=<ltrb 0 0 0 0>
   }
   slot {class=HorizontalBoxSlot, properties {horizontalAlignment=HAlign_Left, verticalAlignment=VAlign_Top, padding=<ltrb 0 6 20 0>}}
  }
  Report_Icon {
   widgetClassPath=&Image
   bIsVariable=true
   parent=Border_69
   widget=Report_Icon
   properties {brush {
     bIsDynamicallyLoaded=false
     drawAs=Image
     imageType=NoImage
     mirroring=NoMirror
     resourceName=None
     resourceObject=null
     tiling=NoTile
     imageSize=<64 64>
     margin=<ltrb 0 0 0 0>
     outlineSettings {
      bUseBrushTransparency=false
      roundingType=HalfHeightRadius
      width=0
      color {colorUseRule=UseColor_Specified, specifiedColor=<rgba 0 0 0 0>}
      cornerRadii=<0 0 0 0>
     }
     tintColor {colorUseRule=UseColor_Specified, specifiedColor=<rgba 1 1 1 1>}
     uVRegion {bIsValid=false, max=<0 0>, min=<0 0>}
    }}
   slot {class=BorderSlot, properties {horizontalAlignment=HAlign_Left, verticalAlignment=VAlign_Top, padding=<ltrb 0 0 0 0>}}
  }
  VerticalBox_97 {
   widgetClassPath=&VerticalBox
   parent=HBHeader
   widget=VerticalBox_97
   slot {class=HorizontalBoxSlot, properties {horizontalAlignment=HAlign_Left, verticalAlignment=VAlign_Top}}
  }
  HeaderText {
   widgetClassPath=&CommonTextBlock
   bIsVariable=true
   parent=VerticalBox_97
   widget=HeaderText
   properties {
    bIsScrollingEnabled=false
    mobileFontSizeMultiplier=1.5
    text="{Header text}"
    visibility=Visible
    wrapTextAt=600
    colorAndOpacity {colorUseRule=UseColor_Specified, specifiedColor=<rgba 0.0331 0.7011 1 1>}
    font {
     bForceMonospaced=false
     bMaterialIsStencil=false
     fontMaterial=null
     fontObject=&BurbankSmall
     letterSpacing=0
     monospacedWidth=1
     size=18
     skewAmount=0
     typefaceFontName=Bold
     outlineSettings {
      bApplyOutlineToDropShadows=false
      bMiteredCorners=false
      bSeparateFillAlpha=false
      outlineMaterial=null
      outlineSize=0
      outlineColor=<rgba 0 0 0 1>
     }
    }
    shadowOffset=<0 0>
   }
   slot {class=VerticalBoxSlot, properties {horizontalAlignment=HAlign_Left, verticalAlignment=VAlign_Top}}
  }
  CurrentStepNameText {
   widgetClassPath=&CommonRichTextBlock
   bIsVariable=true
   parent=VerticalBox_97
   widget=CurrentStepNameText
   properties {
    bIsScrollingEnabled=false
    bOverrideDefaultStyle=true
    mobileTextBlockScale=1.2
    text="{Current Step Name Text}"
    textTransformPolicy=ToUpper
    defaultTextStyleOverride {
     overflowPolicy=Clip
     transformPolicy=ToUpper
     colorAndOpacity {colorUseRule=UseColor_Specified, specifiedColor=<rgba 1 1 1 1>}
     font {
      bForceMonospaced=false
      bMaterialIsStencil=false
      fontMaterial=null
      fontObject=&Burbank
      letterSpacing=0
      monospacedWidth=1
      size=36
      skewAmount=0
      typefaceFontName=BurbankRegular-Black
      outlineSettings {
       bApplyOutlineToDropShadows=false
       bMiteredCorners=false
       bSeparateFillAlpha=false
       outlineMaterial=null
       outlineSize=0
       outlineColor=<rgba 0 0 0 1>
      }
     }
     highlightColor {colorUseRule=UseColor_Specified, specifiedColor=<rgba 0 0 0 1>}
     highlightShape {
      bIsDynamicallyLoaded=false
      drawAs=Image
      imageType=NoImage
      mirroring=NoMirror
      resourceName=None
      resourceObject=null
      tiling=NoTile
      imageSize=<32 32>
      margin=<ltrb 0 0 0 0>
      outlineSettings {
       bUseBrushTransparency=false
       roundingType=HalfHeightRadius
       width=0
       color {colorUseRule=UseColor_Specified, specifiedColor=<rgba 0 0 0 0>}
       cornerRadii=<0 0 0 0>
      }
      tintColor {colorUseRule=UseColor_Specified, specifiedColor=<rgba 1 1 1 1>}
      uVRegion {bIsValid=false, max=<0 0>, min=<0 0>}
     }
     selectedBackgroundColor {colorUseRule=UseColor_Foreground, specifiedColor=<rgba 1 0 1 1>}
     shadowColorAndOpacity=<rgba 0 0 0 1>
     shadowOffset=<0 0>
     strikeBrush {
      bIsDynamicallyLoaded=false
      drawAs=Image
      imageType=NoImage
      mirroring=NoMirror
      resourceName=None
      resourceObject=null
      tiling=NoTile
      imageSize=<32 32>
      margin=<ltrb 0 0 0 0>
      outlineSettings {
       bUseBrushTransparency=false
       roundingType=HalfHeightRadius
       width=0
       color {colorUseRule=UseColor_Specified, specifiedColor=<rgba 0 0 0 0>}
       cornerRadii=<0 0 0 0>
      }
      tintColor {colorUseRule=UseColor_Specified, specifiedColor=<rgba 1 1 1 1>}
      uVRegion {bIsValid=false, max=<0 0>, min=<0 0>}
     }
     underlineBrush {
      bIsDynamicallyLoaded=false
      drawAs=Image
      imageType=NoImage
      mirroring=NoMirror
      resourceName=None
      resourceObject=null
      tiling=NoTile
      imageSize=<32 32>
      margin=<ltrb 0 0 0 0>
      outlineSettings {
       bUseBrushTransparency=false
       roundingType=HalfHeightRadius
       width=0
       color {colorUseRule=UseColor_Specified, specifiedColor=<rgba 0 0 0 0>}
       cornerRadii=<0 0 0 0>
      }
      tintColor {colorUseRule=UseColor_Specified, specifiedColor=<rgba 1 1 1 1>}
      uVRegion {bIsValid=false, max=<0 0>, min=<0 0>}
     }
    }
    renderTransform {angle=0, scale=<1 1>, shear=<0 0>, translation=<0 -6>}
   }
   slot {class=VerticalBoxSlot, properties {horizontalAlignment=HAlign_Left, verticalAlignment=VAlign_Top, padding=<ltrb 0 -2 0 0>}}
  }
  SBContent {
   widgetClassPath=&SizeBox
   parent=VBContent
   widget=SBContent
   properties {
    bOverride_HeightOverride=true
    clipping=ClipToBoundsAlways
    heightOverride=652
    maxDesiredHeight=700
    maxDesiredWidth=600
    minDesiredHeight=700
   }
   slot {class=VerticalBoxSlot, properties {verticalAlignment=VAlign_Top}}
  }
  ListView {
   widgetClassPath=/Script/UMG.ListView
   bIsVariable=true
   parent=SBContent
   widget=ListView
   properties {
    allowOverscroll=false
    bAllowDragging=false
    bEnableRightClickScrolling=false
    bEnableScrollAnimation=true
    bEnableTouchScrolling=false
    bIsFocusable=false
    clipping=Inherit
    consumeMouseWheel=Never
    entryWidgetClass=&/WBP_StepWizard_List_Entry.WBP_StepWizard_List_Entry_C
    fixedLineScrollOffset=0.5
    numDesignerPreviewEntries=1
    selectionMode=None
    wheelScrollMultiplier=0
    scrollBarStyle {
     thickness=0
     draggedThumbImage {
      bIsDynamicallyLoaded=false
      drawAs=NoDrawType
      imageType=FullColor
      mirroring=NoMirror
      resourceName=../../../Engine/Content/Slate/Common/Scrollbar_Thumb.png
      resourceObject=null
      tiling=NoTile
      imageSize=<0 0>
      margin=<ltrb 0.25 0.25 0.25 0.25>
      outlineSettings {
       bUseBrushTransparency=false
       roundingType=HalfHeightRadius
       width=0
       color {colorUseRule=UseColor_Specified, specifiedColor=<rgba 0 0 0 0>}
       cornerRadii=<0 0 0 0>
      }
      tintColor {colorUseRule=UseColor_Specified, specifiedColor=<rgba 1 1 1 1>}
      uVRegion {bIsValid=false, max=<0 0>, min=<0 0>}
     }
     horizontalBackgroundImage {
      bIsDynamicallyLoaded=false
      drawAs=NoDrawType
      imageType=NoImage
      mirroring=NoMirror
      resourceName=None
      resourceObject=null
      tiling=NoTile
      imageSize=<0 0>
      margin=<ltrb 0 0 0 0>
      outlineSettings {
       bUseBrushTransparency=false
       roundingType=HalfHeightRadius
       width=0
       color {colorUseRule=UseColor_Specified, specifiedColor=<rgba 0 0 0 0>}
       cornerRadii=<0 0 0 0>
      }
      tintColor {colorUseRule=UseColor_Specified, specifiedColor=<rgba 1 1 1 1>}
      uVRegion {bIsValid=false, max=<0 0>, min=<0 0>}
     }
     horizontalBottomSlotImage {
      bIsDynamicallyLoaded=false
      drawAs=NoDrawType
      imageType=FullColor
      mirroring=NoMirror
      resourceName=../../../Engine/Content/Slate/Common/Scrollbar_Background_Horizontal.png
      resourceObject=null
      tiling=NoTile
      imageSize=<8 8>
      margin=<ltrb 0 0 0 0>
      outlineSettings {
       bUseBrushTransparency=false
       roundingType=HalfHeightRadius
       width=0
       color {colorUseRule=UseColor_Specified, specifiedColor=<rgba 0 0 0 0>}
       cornerRadii=<0 0 0 0>
      }
      tintColor {colorUseRule=UseColor_Specified, specifiedColor=<rgba 1 1 1 1>}
      uVRegion {bIsValid=false, max=<0 0>, min=<0 0>}
     }
     horizontalTopSlotImage {
      bIsDynamicallyLoaded=false
      drawAs=NoDrawType
      imageType=FullColor
      mirroring=NoMirror
      resourceName=../../../Engine/Content/Slate/Common/Scrollbar_Background_Horizontal.png
      resourceObject=null
      tiling=NoTile
      imageSize=<8 8>
      margin=<ltrb 0 0 0 0>
      outlineSettings {
       bUseBrushTransparency=false
       roundingType=HalfHeightRadius
       width=0
       color {colorUseRule=UseColor_Specified, specifiedColor=<rgba 0 0 0 0>}
       cornerRadii=<0 0 0 0>
      }
      tintColor {colorUseRule=UseColor_Specified, specifiedColor=<rgba 1 1 1 1>}
      uVRegion {bIsValid=false, max=<0 0>, min=<0 0>}
     }
     hoveredThumbImage {
      bIsDynamicallyLoaded=false
      drawAs=NoDrawType
      imageType=FullColor
      mirroring=NoMirror
      resourceName=../../../Engine/Content/Slate/Common/Scrollbar_Thumb.png
      resourceObject=null
      tiling=NoTile
      imageSize=<0 0>
      margin=<ltrb 0.25 0.25 0.25 0.25>
      outlineSettings {
       bUseBrushTransparency=false
       roundingType=HalfHeightRadius
       width=0
       color {colorUseRule=UseColor_Specified, specifiedColor=<rgba 0 0 0 0>}
       cornerRadii=<0 0 0 0>
      }
      tintColor {colorUseRule=UseColor_Specified, specifiedColor=<rgba 1 1 1 1>}
      uVRegion {bIsValid=false, max=<0 0>, min=<0 0>}
     }
     normalThumbImage {
      bIsDynamicallyLoaded=false
      drawAs=NoDrawType
      imageType=FullColor
      mirroring=NoMirror
      resourceName=../../../Engine/Content/Slate/Common/Scrollbar_Thumb.png
      resourceObject=null
      tiling=NoTile
      imageSize=<0 0>
      margin=<ltrb 0.25 0.25 0.25 0.25>
      outlineSettings {
       bUseBrushTransparency=false
       roundingType=HalfHeightRadius
       width=0
       color {colorUseRule=UseColor_Specified, specifiedColor=<rgba 0 0 0 0>}
       cornerRadii=<0 0 0 0>
      }
      tintColor {colorUseRule=UseColor_Specified, specifiedColor=<rgba 1 1 1 1>}
      uVRegion {bIsValid=false, max=<0 0>, min=<0 0>}
     }
     verticalBackgroundImage {
      bIsDynamicallyLoaded=false
      drawAs=NoDrawType
      imageType=NoImage
      mirroring=NoMirror
      resourceName=None
      resourceObject=null
      tiling=NoTile
      imageSize=<0 0>
      margin=<ltrb 0 0 0 0>
      outlineSettings {
       bUseBrushTransparency=false
       roundingType=HalfHeightRadius
       width=0
       color {colorUseRule=UseColor_Specified, specifiedColor=<rgba 0 0 0 0>}
       cornerRadii=<0 0 0 0>
      }
      tintColor {colorUseRule=UseColor_Specified, specifiedColor=<rgba 1 1 1 1>}
      uVRegion {bIsValid=false, max=<0 0>, min=<0 0>}
     }
     verticalBottomSlotImage {
      bIsDynamicallyLoaded=false
      drawAs=NoDrawType
      imageType=FullColor
      mirroring=NoMirror
      resourceName=../../../Engine/Content/Slate/Common/Scrollbar_Background_Vertical.png
      resourceObject=null
      tiling=NoTile
      imageSize=<8 8>
      margin=<ltrb 0 0 0 0>
      outlineSettings {
       bUseBrushTransparency=false
       roundingType=HalfHeightRadius
       width=0
       color {colorUseRule=UseColor_Specified, specifiedColor=<rgba 0 0 0 0>}
       cornerRadii=<0 0 0 0>
      }
      tintColor {colorUseRule=UseColor_Specified, specifiedColor=<rgba 1 1 1 1>}
      uVRegion {bIsValid=false, max=<0 0>, min=<0 0>}
     }
     verticalTopSlotImage {
      bIsDynamicallyLoaded=false
      drawAs=NoDrawType
      imageType=FullColor
      mirroring=NoMirror
      resourceName=../../../Engine/Content/Slate/Common/Scrollbar_Background_Vertical.png
      resourceObject=null
      tiling=NoTile
      imageSize=<8 8>
      margin=<ltrb 0 0 0 0>
      outlineSettings {
       bUseBrushTransparency=false
       roundingType=HalfHeightRadius
       width=0
       color {colorUseRule=UseColor_Specified, specifiedColor=<rgba 0 0 0 0>}
       cornerRadii=<0 0 0 0>
      }
      tintColor {colorUseRule=UseColor_Specified, specifiedColor=<rgba 1 0 0 1>}
      uVRegion {bIsValid=false, max=<0 0>, min=<0 0>}
     }
    }
   }
   slot {class=SizeBoxSlot, properties {padding=<ltrb 120 0 120 0>}}
  }
  BNextStep {
   widgetClassPath=&Border
   bIsVariable=true
   parent=VBContent
   widget=BNextStep
   properties {
    bShowEffectWhenDisabled=false
    horizontalAlignment=HAlign_Left
    verticalAlignment=VAlign_Top
    visibility=Hidden
    background {
     bIsDynamicallyLoaded=false
     drawAs=Image
     imageType=NoImage
     mirroring=NoMirror
     resourceName=None
     resourceObject=/VoiceRep/Materials/MI_UI_LinearGradient
     tiling=NoTile
     imageSize=<32 32>
     margin=<ltrb 0 0 0 0>
     outlineSettings {
      bUseBrushTransparency=false
      roundingType=HalfHeightRadius
      width=0
      color {colorUseRule=UseColor_Specified, specifiedColor=<rgba 0 0 0 0>}
      cornerRadii=<0 0 0 0>
     }
     tintColor {colorUseRule=UseColor_Specified, specifiedColor=<rgba 1 1 1 1>}
     uVRegion {bIsValid=false, max=<0 0>, min=<0 0>}
    }
    padding=<ltrb 120 0 0 0>
   }
   slot {class=VerticalBoxSlot, properties {verticalAlignment=VAlign_Top}}
  }
  HBNextStep {
   widgetClassPath=&HorizontalBox
   parent=BNextStep
   widget=HBNextStep
   slot {class=BorderSlot, properties {horizontalAlignment=HAlign_Left, verticalAlignment=VAlign_Top, padding=<ltrb 120 0 0 0>}}
  }
  Spacer_80 {widgetClassPath=&Spacer, parent=HBNextStep, widget=Spacer_80, properties {size=<0 48>}, slot {class=HorizontalBoxSlot}}
  Step_Image {
   widgetClassPath=&Image
   bIsVariable=true
   parent=HBNextStep
   widget=Step_Image
   properties {brush {
     bIsDynamicallyLoaded=false
     drawAs=Image
     imageType=NoImage
     mirroring=NoMirror
     resourceName=None
     resourceObject=/VoiceRep/Materials/MI_UI_StepDot_Outline
     tiling=NoTile
     imageSize=<16 16>
     margin=<ltrb 0 0 0 0>
     outlineSettings {
      bUseBrushTransparency=false
      roundingType=HalfHeightRadius
      width=0
      color {colorUseRule=UseColor_Specified, specifiedColor=<rgba 0 0 0 0>}
      cornerRadii=<0 0 0 0>
     }
     tintColor {colorUseRule=UseColor_Specified, specifiedColor=<rgba 1 1 1 1>}
     uVRegion {bIsValid=false, max=<0 0>, min=<0 0>}
    }}
   slot {class=HorizontalBoxSlot, properties {horizontalAlignment=HAlign_Left, verticalAlignment=VAlign_Center, padding=<ltrb 0 0 8 0>}}
  }
  Next_Step_Name {
   widgetClassPath=&CommonTextBlock
   bIsVariable=true
   parent=HBNextStep
   widget=Next_Step_Name
   properties {
    bIsScrollingEnabled=false
    mobileFontSizeMultiplier=1.5
    text="{NEXT STEP COUNTER NAME}"
    textTransformPolicy=ToUpper
    visibility=Visible
    wrapTextAt=600
    colorAndOpacity {colorUseRule=UseColor_Specified, specifiedColor=<rgba 0.0331 0.7011 1 1>}
    font {
     bForceMonospaced=false
     bMaterialIsStencil=false
     fontMaterial=null
     fontObject=&BurbankSmall
     letterSpacing=0
     monospacedWidth=1
     size=16
     skewAmount=0
     typefaceFontName=Bold
     outlineSettings {
      bApplyOutlineToDropShadows=false
      bMiteredCorners=false
      bSeparateFillAlpha=false
      outlineMaterial=null
      outlineSize=0
      outlineColor=<rgba 0 0 0 1>
     }
    }
    shadowOffset=<0 0>
   }
   slot {class=HorizontalBoxSlot, properties {horizontalAlignment=HAlign_Left, verticalAlignment=VAlign_Center, padding=<ltrb 8 0 16 0>}}
  }
  CountBoxNext {widgetClassPath=&HorizontalBox, bIsVariable=true, parent=HBNextStep, widget=CountBoxNext, slot {class=HorizontalBoxSlot}}
  NextStepText {
   widgetClassPath=&CommonTextBlock
   bIsVariable=true
   parent=CountBoxNext
   widget=NextStepText
   properties {
    bIsScrollingEnabled=false
    mobileFontSizeMultiplier=1.5
    text=n
    visibility=Visible
    wrapTextAt=600
    colorAndOpacity {colorUseRule=UseColor_Specified, specifiedColor=<rgba 0.0331 0.7011 1 1>}
    font {
     bForceMonospaced=false
     bMaterialIsStencil=false
     fontMaterial=null
     fontObject=&BurbankSmall
     letterSpacing=0
     monospacedWidth=1
     size=16
     skewAmount=0
     typefaceFontName=Bold
     outlineSettings {
      bApplyOutlineToDropShadows=false
      bMiteredCorners=false
      bSeparateFillAlpha=false
      outlineMaterial=null
      outlineSize=0
      outlineColor=<rgba 0 0 0 1>
     }
    }
    shadowOffset=<0 0>
   }
   slot {class=HorizontalBoxSlot, properties {horizontalAlignment=HAlign_Left, verticalAlignment=VAlign_Center}}
  }
  Separator {
   widgetClassPath=&CommonTextBlock
   parent=CountBoxNext
   widget=Separator
   properties {
    bIsScrollingEnabled=false
    mobileFontSizeMultiplier=1.5
    text=/
    visibility=Visible
    wrapTextAt=600
    colorAndOpacity {colorUseRule=UseColor_Specified, specifiedColor=<rgba 0.0331 0.7011 1 1>}
    font {
     bForceMonospaced=false
     bMaterialIsStencil=false
     fontMaterial=null
     fontObject=&BurbankSmall
     letterSpacing=0
     monospacedWidth=1
     size=16
     skewAmount=0
     typefaceFontName=Bold
     outlineSettings {
      bApplyOutlineToDropShadows=false
      bMiteredCorners=false
      bSeparateFillAlpha=false
      outlineMaterial=null
      outlineSize=0
      outlineColor=<rgba 0 0 0 1>
     }
    }
    shadowOffset=<0 0>
   }
   slot {class=HorizontalBoxSlot, properties {horizontalAlignment=HAlign_Left, verticalAlignment=VAlign_Center}}
  }
  TotalStepsText {
   widgetClassPath=&CommonTextBlock
   bIsVariable=true
   parent=CountBoxNext
   widget=TotalStepsText
   properties {
    bIsScrollingEnabled=false
    mobileFontSizeMultiplier=1.5
    text=n
    visibility=Visible
    wrapTextAt=600
    colorAndOpacity {colorUseRule=UseColor_Specified, specifiedColor=<rgba 0.0331 0.7011 1 1>}
    font {
     bForceMonospaced=false
     bMaterialIsStencil=false
     fontMaterial=null
     fontObject=&BurbankSmall
     letterSpacing=0
     monospacedWidth=1
     size=16
     skewAmount=0
     typefaceFontName=Bold
     outlineSettings {
      bApplyOutlineToDropShadows=false
      bMiteredCorners=false
      bSeparateFillAlpha=false
      outlineMaterial=null
      outlineSize=0
      outlineColor=<rgba 0 0 0 1>
     }
    }
    shadowOffset=<0 0>
   }
   slot {class=HorizontalBoxSlot, properties {horizontalAlignment=HAlign_Left, verticalAlignment=VAlign_Center}}
  }
  HB_Buttons {
   widgetClassPath=&HorizontalBox
   bIsVariable=true
   parent=Overlay
   widget=HB_Buttons
   slot {class=OverlaySlot, properties {horizontalAlignment=HAlign_Right, verticalAlignment=VAlign_Bottom, padding=<ltrb 0 26 84 84>}}
  }
  VisibilityBorder_NonTouch {
   widgetClassPath=&CommonHardwareVisibilityBorder
   parent=HB_Buttons
   widget=VisibilityBorder_NonTouch
   properties {visibilityQuery {
     autoDescription=" NONE( Input.Touch )"
     tokenStreamVersion=0
     userDescription=""
     queryTokenStream [0,1,3,1,0]
     tagDictionary [{tagName=Input.Touch}]
    }}
   slot {class=HorizontalBoxSlot, properties {horizontalAlignment=HAlign_Right, verticalAlignment=VAlign_Bottom}}
  }
  Button_Close {
   widgetClassPath=/FortUILibrary/UIKit/Buttons/WBP_UIKit_Button_Quiet
   bIsVariable=true
   parent=VisibilityBorder_NonTouch
   widget=Button_Close
   properties {
    bIsFocusable=false
    text=Close
    visibility=Collapsed
    triggeringInputAction {dataTable=&InputActionDataTable, rowName=Close}
   }
   slot {class=BorderSlot, properties {padding=<ltrb 0 0 0 0>}}
  }
  Button_Back {
   widgetClassPath=&WBP_UIKit_Button_Quiet
   bIsVariable=true
   parent=HB_Buttons
   widget=Button_Back
   properties {
    bIsFocusable=false
    text=Back
    visibility=Collapsed
    triggeringInputAction {dataTable=&InputActionDataTable, rowName=GenericBack}
   }
   slot {class=HorizontalBoxSlot, properties {horizontalAlignment=HAlign_Right, verticalAlignment=VAlign_Bottom}}
  }
  SafeZone_TouchClose {
   widgetClassPath=&SafeZone
   bIsVariable=true
   parent=Overlay
   widget=SafeZone_TouchClose
   slot {class=OverlaySlot, properties {horizontalAlignment=HAlign_Right, padding=<ltrb 0 32 32 0>}}
  }
  VisibilityBorder_Touch {
   widgetClassPath=&CommonHardwareVisibilityBorder
   parent=SafeZone_TouchClose
   widget=VisibilityBorder_Touch
   properties {visibilityQuery {
     autoDescription=" ANY( Input.Touch )"
     tokenStreamVersion=0
     userDescription=""
     queryTokenStream [0,1,1,1,0]
     tagDictionary [{tagName=Input.Touch}]
    }}
   slot {class=SafeZoneSlot}
  }
  Button_TouchClose {
   widgetClassPath=/FortUILibrary/Foundation/Buttons/CloseButton
   bIsVariable=true
   parent=VisibilityBorder_Touch
   widget=Button_TouchClose
   properties {
    onButtonBaseClicked=(&self:WidgetTree.Button_TouchClose.HandleOnButtonClicked)
    onButtonBaseDoubleClicked=(&self:WidgetTree.Button_TouchClose.HandleOnButtonDoubleClicked)
    onButtonBaseHovered=(&self:WidgetTree.Button_TouchClose.HandleOnButtonHovered)
    onButtonBaseUnhovered=(&self:WidgetTree.Button_TouchClose.HandleOnButtonUnhovered)
    onSelectedChangedBase=(&self:WidgetTree.Button_TouchClose.HandleOnSelectedChanged)
   }
   slot {class=BorderSlot, properties {padding=<ltrb 0 0 0 0>}}
  }
 ]
}
