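# Voice report wizard step that tells the player recent voice chat audio goes to Epic, then submits the report.
# Construct shows an audio loading spinner, disables Submit and enlarges elements on mobile.
# ReportIsReadyToSubmit stores the upload result and, via Load Audio, reveals channel participants and focuses Submit.
# Submitting shows a spinner and calls the native parent's SubmitReport; CompleteReport forwards success and stored result to WBP_VoiceReport.
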WidgetBlueprint /VoiceRep/VoiceReportWizard/VoiceReport_Steps/WBP_VoiceRep_Reporting_Step
graphs [EventGraph {nodes [
   n1 = "On Clicked (Button_Submit)"!() -> (OutputDelegate, then, Button:Common Button Base Object Reference) {is=title K2Node_ComponentBoundEvent}
   SetVisibility(execute=[@n1.then, @n6.then], self=Get HBsubmitting() -> (HBsubmitting:Horizontal Box Object Reference), InVisibility=Visible)
   > SetVisibility(self=Get Button_Submit() -> (Button_Submit:WBP UIKit Button Loud Object Reference), InVisibility=Collapsed)
   > SubmitReport(self:Voice Rep Panel Reporting Object Reference)
   n2 = Load Audio!() -> (OutputDelegate, then, Loaded:bool=~) {is=title K2Node_CustomEvent}
   > n3 = Branch(Condition=@n2.Loaded)
   > SetVisibility(self=Get HBLoadingAudio() -> (HBLoadingAudio:Horizontal Box Object Reference), InVisibility=Collapsed)
   > SetVisibility(self=[Get HB_Party() -> (HB_Party:Horizontal Box Object Reference), Get Button_Submit()], InVisibility=Visible)
   n4 = Get Button_Submit()
   > SetIsEnabled(self=@n4, bInIsEnabled=true)
   > SetKeyboardFocus(self:Widget Object Reference=@n4)
   SetVisibility(execute=[@n7.then, @n3.else], self=Get HBLoadingAudio(), InVisibility=Visible)
   > SetIsEnabled(self=Get Button_Submit(), bInIsEnabled=false)
   n5 = IsMobileGame()
   > SetUserSpecifiedScale!(
    execute
    self:Scale Box Object Reference=Get ScaleBox_TextBlock_Mobile() -> (ScaleBox_TextBlock_Mobile:Scale Box Object Reference)
    InUserSpecifiedScale:float=Select(Option 0:float=1, Option 1:float=1.4, Index=@n5) -> (ReturnValue:float)
   ) -> (then)
   > SetDesiredSizeOverride!(
    execute
    self:Image Object Reference=Get Step_Header_1() -> (Step_Header_1:Image Object Reference)
    DesiredSize:Vector 2D Structure=Select(Option 0:Vector 2D Structure=(X=32,Y=32), Option 1:Vector 2D Structure=(X=38,Y=38), Index=@n5) -> (ReturnValue:Vector 2D Structure)
   ) -> (then)
   > SetDesiredSizeOverride(
    self=Get ProgressSpinner() -> (ProgressSpinner:Image Object Reference)
    DesiredSize=Select(Option 0=(X=28,Y=28), Option 1=(X=38,Y=38), Index=@n5)
   )
   > SetUserSpecifiedScale(
    self=Get ScaleBox_Portraits_Mobile() -> (ScaleBox_Portraits_Mobile:Scale Box Object Reference)
    InUserSpecifiedScale=Select(Option 0:float=1, Option 1:float=1.3, Index=@n5) -> (ReturnValue:float)
   )
   > SetPadding!(
    execute
    self:Horizontal Box Slot Object Reference=SlotAsHorizontalBoxSlot!(Widget:Widget Object Reference=Get Line() -> (Line:Image Object Reference)) -> (ReturnValue:Horizontal Box Slot Object Reference)
    InPadding_Left:float=Select(Option 0=7, Option 1=11, Index=@n5)
    InPadding_Top:float=0
    InPadding_Right:float=80
    InPadding_Bottom:float=0
   ) -> (then)
   n6 = OnSubmitReport() {is=title K2Node_CustomEvent}
   n7 = Event Construct()
   n8 = Event CompleteReport!() -> (OutputDelegate, then, bSucceeded:bool) {is=title K2Node_Event}
   > n9 = FindParentWidgetOfType!(
    execute
    StartingWidget:Widget Object Reference=Self-Reference()
    Type:Widget Class Reference=/VoiceRep/VoiceReportWizard/WBP_VoiceReport.WBP_VoiceReport_C
   ) -> (then, ReturnValue:WBP Voice Report Object Reference)
   > Complete Report!(
    execute
    self:WBP Voice Report Object Reference=@n9
    Sent:bool=@n8.bSucceeded
    Result:EVoiceRepUploadRecordResult Enum=Get Result() -> (Result:EVoiceRepUploadRecordResult Enum=Success)
   ) -> (then)
   n10 = Event ReportIsReadyToSubmit!() -> (OutputDelegate, then, Result:EVoiceRepUploadRecordResult Enum=Success) {is=title K2Node_Event}
   > n11 = Set Result(Result:EVoiceRepUploadRecordResult Enum=@n10.Result) -> (Output_Get:EVoiceRepUploadRecordResult Enum=Success)
   > Switch on EVoiceRepUploadRecordResult!(execute, Selection:EVoiceRepUploadRecordResult Enum=@n11) -> (Success:Exec, Cancelled:Exec, NoAudioToUpload:Exec, EmeraldError:Exec, FailedUpload:Exec) {is=title K2Node_SwitchEnum}
   > Load Audio!(execute, self, Loaded:bool=true) -> (then) {is=title K2Node_CallFunction}
  ]}]
variables [
 Parent Widget {type=Widget, default=null}
 Result {type=EVoiceRepUploadRecordResult, default=Success}
]
widgets {
 info {inheritedWidgetCount=2, parentClass=/Script/FortniteUI.VoiceRepPanel_Reporting, rootWidgetClass=&SizeBox, widgetCount=31}
 widgets [
  SizeBox_95 {
   widgetClassPath=&SizeBox
   widget=SizeBox_95
   properties {
    bOverride_HeightOverride=true
    heightOverride=600
    maxDesiredHeight=700
    maxDesiredWidth=690
    minDesiredHeight=670
    minDesiredWidth=1000
    renderTransformPivot=<0.5 0>
   }
  }
  VerticalBox_101 {widgetClassPath=&VerticalBox, parent=SizeBox_95, widget=VerticalBox_101, slot {class=SizeBoxSlot}}
  HorizontalBox_375 {
   widgetClassPath=&HorizontalBox
   parent=VerticalBox_101
   widget=HorizontalBox_375
   slot {class=VerticalBoxSlot, properties {size {sizeRule=Fill, value=1}}}
  }
  Line {
   widgetClassPath=&Image
   bIsVariable=true
   parent=HorizontalBox_375
   widget=Line
   properties {brush {
     bIsDynamicallyLoaded=false
     drawAs=Image
     imageType=NoImage
     mirroring=NoMirror
     resourceName=None
     resourceObject=null
     tiling=NoTile
     imageSize=<2 5>
     margin=<ltrb 0 0 0 0>
     outlineSettings {
      bUseBrushTransparency=false
      roundingType=HalfHeightRadius
      width=0
      color {colorUseRule=UseColor_Specified, specifiedColor=<rgba 0 0 0 0>}
      cornerRadii=<0 0 0 0>
     }
     tintColor {colorUseRule=UseColor_Specified, specifiedColor=<rgba 1 1 1 1>}
     uVRegion {bIsValid=false, max=<0 0>, min=<0 0>}
    }}
   slot {class=HorizontalBoxSlot, properties {padding=<ltrb 7 0 80 0>}}
  }
  AthenaScrollBox_Content {
   widgetClassPath=/Script/FortniteUI.AthenaScrollBox
   bInherited=true
   parent=HorizontalBox_375
   widget=AthenaScrollBox_Content
   properties {allowOverscroll=false, scrollBarVisibility=Collapsed}
   slot {class=HorizontalBoxSlot, properties {padding=<ltrb 0 64 0 0>, size {sizeRule=Fill, value=1}}}
  }
  VerticalBox_62 {
   widgetClassPath=&VerticalBox
   parent=AthenaScrollBox_Content
   widget=VerticalBox_62
   slot {class=ScrollBoxSlot, properties {padding=<ltrb 15 0 0 0>}}
  }
  ScaleBox_TextBlock_Mobile {
   widgetClassPath=&ScaleBox
   bIsVariable=true
   parent=VerticalBox_62
   widget=ScaleBox_TextBlock_Mobile
   properties {stretch=UserSpecified}
   slot {class=VerticalBoxSlot, properties {horizontalAlignment=HAlign_Left, verticalAlignment=VAlign_Top, padding=<ltrb 0 0 0 32>}}
  }
  UEFN_TextBlock_C_306 {
   widgetClassPath=&UEFN_TextBlock
   bIsVariable=true
   parent=ScaleBox_TextBlock_Mobile
   widget=UEFN_TextBlock_C_306
   properties {
    flowDirectionPreference=Culture
    text="Your report will send the last 5 minutes of voice chat audio to Epic Games for review. Epic will review the conduct of all participants in the conversation, including you. Sanctioned players will be notified if any actions are taken based on your report. Epic will discard the recording once it is no longer needed for enforcement or legal purposes. Players will not be notified that you reported them."
    wrapTextAt=1000
    colorAndOpacity {colorUseRule=UseColor_Specified, specifiedColor=<rgba 0.5583 0.8879 1 1>}
    font {
     bForceMonospaced=false
     bMaterialIsStencil=false
     fontMaterial=null
     fontObject=&BurbankSmall
     letterSpacing=0
     monospacedWidth=1
     size=18
     skewAmount=0
     typefaceFontName=Bold
     outlineSettings {
      bApplyOutlineToDropShadows=false
      bMiteredCorners=false
      bSeparateFillAlpha=false
      outlineMaterial=null
      outlineSize=0
      outlineColor=<rgba 0 0 0 1>
     }
    }
   }
   slot {class=ScaleBoxSlot, properties {horizontalAlignment=HAlign_Left}}
  }
  HorizontalBox_186 {
   widgetClassPath=&HorizontalBox
   parent=VerticalBox_62
   widget=HorizontalBox_186
   slot {class=VerticalBoxSlot, properties {horizontalAlignment=HAlign_Left}}
  }
  HBLoadingAudio {
   widgetClassPath=&HorizontalBox
   bIsVariable=true
   parent=HorizontalBox_186
   widget=HBLoadingAudio
   properties {visibility=Visible}
   slot {class=HorizontalBoxSlot, properties {horizontalAlignment=HAlign_Left, verticalAlignment=VAlign_Top}}
  }
  ProgressSpinner_1 {
   widgetClassPath=&Image
   bIsVariable=true
   parent=HBLoadingAudio
   widget=ProgressSpinner_1
   properties {
    brush {
     bIsDynamicallyLoaded=false
     drawAs=Image
     imageType=NoImage
     mirroring=NoMirror
     resourceName=None
     resourceObject=/FortUILibrary/Systems/Throbber/M_UIKit_Throbber
     tiling=NoTile
     imageSize=<30 30>
     margin=<ltrb 0 0 0 0>
     outlineSettings {
      bUseBrushTransparency=false
      roundingType=HalfHeightRadius
      width=0
      color {colorUseRule=UseColor_Specified, specifiedColor=<rgba 0 0 0 0>}
      cornerRadii=<0 0 0 0>
     }
     tintColor {colorUseRule=UseColor_Specified, specifiedColor=<rgba 1 1 1 1>}
     uVRegion {bIsValid=false, max=<0 0>, min=<0 0>}
    }
    colorAndOpacity=<rgba 0.0331 0.7011 1 1>
   }
   slot {class=HorizontalBoxSlot, properties {horizontalAlignment=HAlign_Center, verticalAlignment=VAlign_Center}}
  }
  Text_Loading_State_1 {
   widgetClassPath=&CommonTextBlock
   bIsVariable=true
   parent=HBLoadingAudio
   widget=Text_Loading_State_1
   properties {
    bIsScrollingEnabled=false
    mobileFontSizeMultiplier=1.5
    text=Loading voice chat audio...
    visibility=Visible
    wrapTextAt=600
    colorAndOpacity {colorUseRule=UseColor_Specified, specifiedColor=<rgba 0.0331 0.7011 1 1>}
    font {
     bForceMonospaced=false
     bMaterialIsStencil=false
     fontMaterial=null
     fontObject=&BurbankSmall
     letterSpacing=0
     monospacedWidth=1
     size=15
     skewAmount=0
     typefaceFontName=Bold
     outlineSettings {
      bApplyOutlineToDropShadows=false
      bMiteredCorners=false
      bSeparateFillAlpha=false
      outlineMaterial=null
      outlineSize=0
      outlineColor=<rgba 0 0 0 1>
     }
    }
    shadowOffset=<0 0>
   }
   slot {class=HorizontalBoxSlot, properties {horizontalAlignment=HAlign_Left, verticalAlignment=VAlign_Center, padding=<ltrb 10 0 10 0>}}
  }
  HB_Party {
   widgetClassPath=&HorizontalBox
   bIsVariable=true
   parent=HorizontalBox_186
   widget=HB_Party
   properties {visibility=Hidden}
   slot {class=HorizontalBoxSlot, properties {horizontalAlignment=HAlign_Left, verticalAlignment=VAlign_Top}}
  }
  Image_104 {
   widgetClassPath=&Image
   bIsVariable=true
   parent=HB_Party
   widget=Image_104
   properties {
    brush {
     bIsDynamicallyLoaded=false
     drawAs=Image
     imageType=NoImage
     mirroring=NoMirror
     resourceName=None
     resourceObject=null
     tiling=NoTile
     imageSize=<6 150>
     margin=<ltrb 0 0 0 0>
     outlineSettings {
      bUseBrushTransparency=false
      roundingType=HalfHeightRadius
      width=0
      color {colorUseRule=UseColor_Specified, specifiedColor=<rgba 0 0 0 0>}
      cornerRadii=<0 0 0 0>
     }
     tintColor {colorUseRule=UseColor_Specified, specifiedColor=<rgba 1 1 1 1>}
     uVRegion {bIsValid=false, max=<0 0>, min=<0 0>}
    }
    colorAndOpacity=<rgba 0.107 0.8879 0.8963 1>
   }
   slot {class=HorizontalBoxSlot, properties {padding=<ltrb 0 0 14 0>}}
  }
  VerticalBox_153 {widgetClassPath=&VerticalBox, parent=HB_Party, widget=VerticalBox_153, slot {class=HorizontalBoxSlot}}
  Text_Label {
   widgetClassPath=&CommonTextBlock
   parent=VerticalBox_153
   widget=Text_Label
   properties {
    bIsScrollingEnabled=false
    mobileFontSizeMultiplier=1.5
    text=voice channel participants
    textTransformPolicy=ToUpper
    visibility=Visible
    wrapTextAt=600
    colorAndOpacity {colorUseRule=UseColor_Specified, specifiedColor=<rgba 0.0331 0.7011 1 1>}
    font {
     bForceMonospaced=false
     bMaterialIsStencil=false
     fontMaterial=null
     fontObject=&BurbankSmall
     letterSpacing=0
     monospacedWidth=1
     size=18
     skewAmount=0
     typefaceFontName=Bold
     outlineSettings {
      bApplyOutlineToDropShadows=false
      bMiteredCorners=false
      bSeparateFillAlpha=false
      outlineMaterial=null
      outlineSize=0
      outlineColor=<rgba 0 0 0 1>
     }
    }
    shadowOffset=<0 0>
   }
   slot {class=VerticalBoxSlot, properties {horizontalAlignment=HAlign_Left, verticalAlignment=VAlign_Top, padding=<ltrb 0 0 0 10>}}
  }
  ScaleBox_Portraits_Mobile {
   widgetClassPath=&ScaleBox
   bIsVariable=true
   parent=VerticalBox_153
   widget=ScaleBox_Portraits_Mobile
   properties {stretch=UserSpecified}
   slot {class=VerticalBoxSlot, properties {horizontalAlignment=HAlign_Left, verticalAlignment=VAlign_Top}}
  }
  EntryBox_ChannelPortraits {
   widgetClassPath=&DynamicEntryBox
   bInherited=true
   bIsVariable=true
   parent=ScaleBox_Portraits_Mobile
   widget=EntryBox_ChannelPortraits
   properties {
    entryBoxType=Overlay
    entryHorizontalAlignment=HAlign_Left
    entryVerticalAlignment=VAlign_Top
    entryWidgetClass=/Game/UI/Sidebar/SidebarPartyPortrait.SidebarPartyPortrait_C
    entrySpacing=<-60 0>
   }
   slot {class=ScaleBoxSlot, properties {horizontalAlignment=HAlign_Left}}
  }
  Border_257 {
   widgetClassPath=&Border
   parent=VerticalBox_153
   widget=Border_257
   properties {horizontalAlignment=HAlign_Left, brushColor=<rgba 1 1 1 0.1>, padding=<ltrb 0 0 0 0>}
   slot {class=VerticalBoxSlot, properties {verticalAlignment=VAlign_Bottom, padding=<ltrb 0 16 0 0>, size {sizeRule=Fill, value=1}}}
  }
  HorizontalBox_262 {
   widgetClassPath=&HorizontalBox
   parent=Border_257
   widget=HorizontalBox_262
   slot {class=BorderSlot, properties {horizontalAlignment=HAlign_Left, padding=<ltrb 0 0 0 0>}}
  }
  Step_Header_1 {
   widgetClassPath=&Image
   bIsVariable=true
   parent=HorizontalBox_262
   widget=Step_Header_1
   properties {brush {
     bIsDynamicallyLoaded=false
     drawAs=Image
     imageType=NoImage
     mirroring=NoMirror
     resourceName=None
     resourceObject=/VoiceRep/Materials/MI_UI_Icon_Paperclip
     tiling=NoTile
     imageSize=<30 30>
     margin=<ltrb 0 0 0 0>
     outlineSettings {
      bUseBrushTransparency=false
      roundingType=HalfHeightRadius
      width=0
      color {colorUseRule=UseColor_Specified, specifiedColor=<rgba 0 0 0 0>}
      cornerRadii=<0 0 0 0>
     }
     tintColor {colorUseRule=UseColor_Specified, specifiedColor=<rgba 1 1 1 1>}
     uVRegion {bIsValid=false, max=<0 0>, min=<0 0>}
    }}
   slot {class=HorizontalBoxSlot, properties {horizontalAlignment=HAlign_Left, verticalAlignment=VAlign_Center, padding=<ltrb 8 0 0 0>}}
  }
  Text_Label_2 {
   widgetClassPath=&CommonTextBlock
   parent=HorizontalBox_262
   widget=Text_Label_2
   properties {
    bIsScrollingEnabled=false
    mobileFontSizeMultiplier=1.5
    text=Voice Chat audio up to last 5 minutes
    visibility=Visible
    wrapTextAt=600
    colorAndOpacity {colorUseRule=UseColor_Specified, specifiedColor=<rgba 0.0319 0.7011 1 1>}
    font {
     bForceMonospaced=false
     bMaterialIsStencil=false
     fontMaterial=null
     fontObject=&BurbankSmall
     letterSpacing=0
     monospacedWidth=1
     size=15
     skewAmount=0
     typefaceFontName=Bold
     outlineSettings {
      bApplyOutlineToDropShadows=false
      bMiteredCorners=false
      bSeparateFillAlpha=false
      outlineMaterial=null
      outlineSize=0
      outlineColor=<rgba 0 0 0 1>
     }
    }
    shadowOffset=<0 0>
   }
   slot {class=HorizontalBoxSlot, properties {horizontalAlignment=HAlign_Left, verticalAlignment=VAlign_Center, padding=<ltrb 2 0 16 0>}}
  }
  Spacer_114 {widgetClassPath=&Spacer, parent=HorizontalBox_262, widget=Spacer_114, properties {size=<0 40>}, slot {class=HorizontalBoxSlot}}
  Spacer_108 {widgetClassPath=&Spacer, parent=HB_Party, widget=Spacer_108, properties {size=<0 170>}, slot {class=HorizontalBoxSlot}}
  HorizontalBox_179 {
   widgetClassPath=&HorizontalBox
   parent=VerticalBox_62
   widget=HorizontalBox_179
   slot {class=VerticalBoxSlot, properties {horizontalAlignment=HAlign_Left, verticalAlignment=VAlign_Top, padding=<ltrb 0 40 0 0>}}
  }
  Button_Submit {
   widgetClassPath=/FortUILibrary/UIKit/Buttons/WBP_UIKit_Button_Loud
   bIsVariable=true
   parent=HorizontalBox_179
   widget=Button_Submit
   properties {
    bApplyAlphaOnDisable=true
    bIsEnabled=false
    minHeight=56
    text=SUBMIT
    visibility=Visible
    triggeringInputAction {dataTable=&InputActionDataTable, rowName=GenericAccept}
   }
   slot {class=HorizontalBoxSlot, properties {horizontalAlignment=HAlign_Left, verticalAlignment=VAlign_Center}}
  }
  HBsubmitting {
   widgetClassPath=&HorizontalBox
   bIsVariable=true
   parent=HorizontalBox_179
   widget=HBsubmitting
   properties {visibility=Collapsed}
   slot {class=HorizontalBoxSlot, properties {horizontalAlignment=HAlign_Left, verticalAlignment=VAlign_Center}}
  }
  Spacer_233 {widgetClassPath=&Spacer, parent=HBsubmitting, widget=Spacer_233, properties {size=<0 56>}, slot {class=HorizontalBoxSlot}}
  ProgressSpinner {
   widgetClassPath=&Image
   bIsVariable=true
   parent=HBsubmitting
   widget=ProgressSpinner
   properties {
    brush {
     bIsDynamicallyLoaded=false
     drawAs=Image
     imageType=NoImage
     mirroring=NoMirror
     resourceName=None
     resourceObject=&M_UIKit_Throbber
     tiling=NoTile
     imageSize=<30 30>
     margin=<ltrb 0 0 0 0>
     outlineSettings {
      bUseBrushTransparency=false
      roundingType=HalfHeightRadius
      width=0
      color {colorUseRule=UseColor_Specified, specifiedColor=<rgba 0 0 0 0>}
      cornerRadii=<0 0 0 0>
     }
     tintColor {colorUseRule=UseColor_Specified, specifiedColor=<rgba 1 1 1 1>}
     uVRegion {bIsValid=false, max=<0 0>, min=<0 0>}
    }
    colorAndOpacity=<rgba 0.0331 0.7011 1 1>
   }
   slot {class=HorizontalBoxSlot, properties {horizontalAlignment=HAlign_Left, verticalAlignment=VAlign_Center}}
  }
  Text_Loading_State {
   widgetClassPath=&CommonTextBlock
   bIsVariable=true
   parent=HBsubmitting
   widget=Text_Loading_State
   properties {
    bIsScrollingEnabled=false
    mobileFontSizeMultiplier=1.5
    text=Submitting report. Please do not close this window.
    visibility=Visible
    wrapTextAt=600
    colorAndOpacity {colorUseRule=UseColor_Specified, specifiedColor=<rgba 0.0331 0.7011 1 1>}
    font {
     bForceMonospaced=false
     bMaterialIsStencil=false
     fontMaterial=null
     fontObject=&BurbankSmall
     letterSpacing=0
     monospacedWidth=1
     size=15
     skewAmount=0
     typefaceFontName=Bold
     outlineSettings {
      bApplyOutlineToDropShadows=false
      bMiteredCorners=false
      bSeparateFillAlpha=false
      outlineMaterial=null
      outlineSize=0
      outlineColor=<rgba 0 0 0 1>
     }
    }
    shadowOffset=<0 0>
   }
   slot {class=HorizontalBoxSlot, properties {horizontalAlignment=HAlign_Left, verticalAlignment=VAlign_Center, padding=<ltrb 10 0 10 0>}}
  }
  Spacer_335 {widgetClassPath=&Spacer, parent=VerticalBox_62, widget=Spacer_335, properties {size=<1 10>}, slot {class=VerticalBoxSlot}}
 ]
}
