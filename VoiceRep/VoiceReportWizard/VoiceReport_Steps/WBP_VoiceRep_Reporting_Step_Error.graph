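# Error step screen for the voice report wizard, a CommonActivatableWidget telling the player that the report could not be submitted.
# A SizeBox holds a vertical stack with an uppercase "Something Went Wrong" rich-text header above a retry-later message.
# All content is static, with no variables.
# The event graph holds only unwired nodes, including a 1-key event and calls to create WBP_VoiceReport, add it to viewport and load a step.
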WidgetBlueprint /VoiceRep/VoiceReportWizard/VoiceReport_Steps/WBP_VoiceRep_Reporting_Step_Error
graphs [EventGraph {nodes [
   AddToViewport!(execute, self:User Widget Object Reference, ZOrder:int=0) -> (then)
   Load Current Step!(execute, self:WBP Voice Report Object Reference, Step:int=0) -> (then)
   Create Widget!(
    execute:Exec
    Class:User Widget Class Reference=/VoiceRep/VoiceReportWizard/WBP_VoiceReport.WBP_VoiceReport_C
    OwningPlayer:Player Controller Object Reference
   ) -> (then:Exec, ReturnValue:WBP Voice Report Object Reference)
   Event PreConstruct()
   Event Construct()
   Event Tick()
   1!() -> (Pressed:Exec, Released:Exec, Key:Key Structure) {is=title K2Node_InputKey, inputKey=One}
  ]}]
variables []
widgets {
 info {inheritedWidgetCount=0, parentClass=&CommonActivatableWidget, rootWidgetClass=&SizeBox, widgetCount=4}
 widgets [
  SizeBox_95 {
   widgetClassPath=&SizeBox
   widget=SizeBox_95
   properties {bOverride_MinDesiredHeight=true, maxDesiredHeight=700, minDesiredHeight=600, minDesiredWidth=1000}
  }
  VerticalBox_82 {
   widgetClassPath=&VerticalBox
   parent=SizeBox_95
   widget=VerticalBox_82
   slot {class=SizeBoxSlot, properties {horizontalAlignment=HAlign_Center, verticalAlignment=VAlign_Center}}
  }
  Header_1 {
   widgetClassPath=&CommonRichTextBlock
   parent=VerticalBox_82
   widget=Header_1
   properties {
    bIsScrollingEnabled=false
    bOverrideDefaultStyle=true
    justification=Center
    text=Something Went Wrong
    textTransformPolicy=ToUpper
    defaultTextStyleOverride {
     overflowPolicy=Clip
     transformPolicy=ToUpper
     colorAndOpacity {colorUseRule=UseColor_Specified, specifiedColor=<rgba 1 1 1 1>}
     font {
      bForceMonospaced=false
      bMaterialIsStencil=false
      fontMaterial=null
      fontObject=&Burbank
      letterSpacing=0
      monospacedWidth=1
      size=36
      skewAmount=0
      typefaceFontName=BurbankRegular-Black
      outlineSettings {
       bApplyOutlineToDropShadows=false
       bMiteredCorners=false
       bSeparateFillAlpha=false
       outlineMaterial=null
       outlineSize=0
       outlineColor=<rgba 0 0 0 1>
      }
     }
     highlightColor {colorUseRule=UseColor_Specified, specifiedColor=<rgba 0 0 0 1>}
     highlightShape {
      bIsDynamicallyLoaded=false
      drawAs=Image
      imageType=NoImage
      mirroring=NoMirror
      resourceName=None
      resourceObject=null
      tiling=NoTile
      imageSize=<32 32>
      margin=<ltrb 0 0 0 0>
      outlineSettings {
       bUseBrushTransparency=false
       roundingType=HalfHeightRadius
       width=0
       color {colorUseRule=UseColor_Specified, specifiedColor=<rgba 0 0 0 0>}
       cornerRadii=<0 0 0 0>
      }
      tintColor {colorUseRule=UseColor_Specified, specifiedColor=<rgba 1 1 1 1>}
      uVRegion {bIsValid=false, max=<0 0>, min=<0 0>}
     }
     selectedBackgroundColor {colorUseRule=UseColor_Foreground, specifiedColor=<rgba 1 0 1 1>}
     shadowColorAndOpacity=<rgba 0 0 0 1>
     shadowOffset=<0 0>
     strikeBrush {
      bIsDynamicallyLoaded=false
      drawAs=Image
      imageType=NoImage
      mirroring=NoMirror
      resourceName=None
      resourceObject=null
      tiling=NoTile
      imageSize=<32 32>
      margin=<ltrb 0 0 0 0>
      outlineSettings {
       bUseBrushTransparency=false
       roundingType=HalfHeightRadius
       width=0
       color {colorUseRule=UseColor_Specified, specifiedColor=<rgba 0 0 0 0>}
       cornerRadii=<0 0 0 0>
      }
      tintColor {colorUseRule=UseColor_Specified, specifiedColor=<rgba 1 1 1 1>}
      uVRegion {bIsValid=false, max=<0 0>, min=<0 0>}
     }
     underlineBrush {
      bIsDynamicallyLoaded=false
      drawAs=Image
      imageType=NoImage
      mirroring=NoMirror
      resourceName=None
      resourceObject=null
      tiling=NoTile
      imageSize=<32 32>
      margin=<ltrb 0 0 0 0>
      outlineSettings {
       bUseBrushTransparency=false
       roundingType=HalfHeightRadius
       width=0
       color {colorUseRule=UseColor_Specified, specifiedColor=<rgba 0 0 0 0>}
       cornerRadii=<0 0 0 0>
      }
      tintColor {colorUseRule=UseColor_Specified, specifiedColor=<rgba 1 1 1 1>}
      uVRegion {bIsValid=false, max=<0 0>, min=<0 0>}
     }
    }
    renderTransform {angle=0, scale=<1 1>, shear=<0 0>, translation=<0 -6>}
   }
   slot {class=VerticalBoxSlot, properties {horizontalAlignment=HAlign_Center, verticalAlignment=VAlign_Center, padding=<ltrb 0 0 0 10>}}
  }
  UEFN_TextBlock_C_306 {
   widgetClassPath=&UEFN_TextBlock
   parent=VerticalBox_82
   widget=UEFN_TextBlock_C_306
   properties {
    justification=Center
    text=Unable to submit your report at this time. Please try again later.
    colorAndOpacity {colorUseRule=UseColor_Specified, specifiedColor=<rgba 0.5583 0.8879 1 1>}
    font {
     bForceMonospaced=false
     bMaterialIsStencil=false
     fontMaterial=null
     fontObject=&BurbankSmall
     letterSpacing=0
     monospacedWidth=1
     size=18
     skewAmount=0
     typefaceFontName=Bold
     outlineSettings {
      bApplyOutlineToDropShadows=false
      bMiteredCorners=false
      bSeparateFillAlpha=false
      outlineMaterial=null
      outlineSize=0
      outlineColor=<rgba 0 0 0 1>
     }
    }
   }
   slot {class=VerticalBoxSlot, properties {horizontalAlignment=HAlign_Center, verticalAlignment=VAlign_Center, padding=<ltrb 0 0 0 10>}}
  }
 ]
}
